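# Niagara sprite effect for the hooded variant of the Tour Bus cosmetic's hair, adding flames and a glow decal.
# Flames spawns at a User.SpawnRate-scaled rate, while Hair_Decal bursts fading sprites.
# Both resample skeletal mesh sockets every update with random offsets, keeping sprites attached.
# User.VariantID selects cyan or orange/gold, FNTimeOfDayColor tints at night, and User.bisHeadHidden sets each particle's VisibilityTag.
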NiagaraSystem /BRCosmetics/Accessories/Hats/M_MED_Tour_Bus_FaceAcc/FX/NS_TourBus_Hair_Hooded
dependencies {
 usedDataInterfaces [&NiagaraDataInterfaceCurve, &NiagaraDataInterfaceSkeletalMesh]
 usedDynamicInputs []
 usedModules [
  &SystemState
  &self:SystemSpawnScript.SystemScriptSource.SystemScriptGraph.NiagaraNodeAssignment_0.SetVariables_#1
 ]
}
emitters {
 Flames {
  emitterUpdateScript [
   &EmitterState()
   &SpawnRate(
    SpawnRate=&Multiply_Float(A=&RandomRangeFloat(Minimum=1, Maximum=3), B=$User.SpawnRate)
    Use Spawn Probability=false
   )
  ]
  particleSpawnScript [
   &InitializeParticle(
    Color=&MakeCustomLinearColorFromBool(
     Bool=&SetBoolByFloatComparison(A=$User.VariantID, B=0, Comparison Type=::NewEnumerator4)
     True Color=<rgba 0 0.566 0.7605 1>
     False Color=<rgba 0.761 0.4265 0 1>
    )
    Lifetime Min=0.7
    Lifetime Mode=::NewEnumerator1
    Sprite Size Mode=::NewEnumerator1
    Uniform Sprite Size Max=50
    Uniform Sprite Size Min=25
    UsePositionOffset=false
    Write Material Random=true
   )
   &SkeletalMeshLocation(
    Bone / Socket Index=2
    Bone Sampling Mode=::NewEnumerator4
    MaskSpawnGroup=false
    Sampled Position Offset=<-3 0 0>
    Skeletal Mesh=$System.Skeletal Mesh
   ) {enabled=false}
   &ShapeLocation(MaskSpawnGroup=false, Override Seed=false, Sphere Radius=0.8, UseOwnerScale=true) {enabled=false}
   &AddVelocity(Velocity=&RandomRangeVector(Minimum=<-2.4 0.5 1>, Maximum=<-6 1 2.5>)) {enabled=false}
   &CameraOffset(Camera Offset Amount=8)
  ]
  particleUpdateScript [
   &ParticleState()
   &SkeletalMeshLocation(
    MaskSpawnGroup=false
    Mesh Sampling Type=::NewEnumerator5
    Sampled Position Offset=&RandomRangeVector(Minimum=<-2.5 0 0>, Maximum=<0.6 0 0>)
    Skeletal Mesh=$System.Skeletal Mesh
    Socket Sampling Mode=::NewEnumerator5
   ) {module=SkeletalMeshLocation001}
   &GravityForce(Gravity=&RandomRangeVector(Minimum=<-2.4 0.5 1>, Maximum=<-6 1 2.5>))
   &ScaleSpriteSize(Uniform Curve Sprite Scale=DataInterface() {propertyValues {Curve {keys [
        <at 0 0.1035 RCIM_Cubic>
        {arriveTangent=-0.0049, interpMode=RCIM_Cubic, leaveTangent=-0.0049, time=0.5835, value=0.4}
        {arriveTangent=-1.8945, interpMode=RCIM_Cubic, leaveTangent=-1.8945, tangentMode=RCTM_User, time=1, value=0.0985}
       ]}}})
   &ScaleColor(
    Scale Alpha=&FloatFromCurve(
     FloatCurve=DataInterface() {propertyValues {Curve {keys [
         {arriveTangent=-5.2561, interpMode=RCIM_Cubic, leaveTangent=-5.2561, tangentMode=RCTM_User, time=0, value=1}
         {arriveTangentWeight=0.0674, interpMode=RCIM_Cubic, leaveTangentWeight=0.2659, tangentMode=RCTM_User, time=0.2022, value=0.3}
         {arriveTangent=1.5683, interpMode=RCIM_Cubic, leaveTangent=1.5683, tangentMode=RCTM_User, time=1, value=1}
        ]}}}
     Scale Curve=$User.Alpha
    )
    Scale RGB=&VectorFromFloat(Value=&FloatFromCurve(FloatCurve=DataInterface() {propertyValues {Curve {keys [
          {arriveTangent=10.8365, interpMode=RCIM_Cubic, leaveTangent=10.8365, tangentMode=RCTM_User, time=0.2, value=2.5}
          {arriveTangentWeight=0.2028, interpMode=RCIM_Cubic, leaveTangentWeight=0.5152, tangentMode=RCTM_User, time=0.4259, value=4}
          {arriveTangent=-8.9073, interpMode=RCIM_Cubic, leaveTangent=-8.9073, tangentMode=RCTM_User, time=1, value=0.5}
         ]}}}))
    ScaleA=true
    ScaleRGB=true
   )
   &SolveForcesAndVelocity(Clamp Velocity=false, Limit Acceleration=false)
   &DynamicMaterialParameters(
    Index 0 Param 1=&FloatFromCurve(FloatCurve=DataInterface() {propertyValues {Curve {keys [{arriveTangent=2.6128, interpMode=RCIM_Cubic, leaveTangent=2.6128, tangentMode=RCTM_User, time=0, value=0.0108}, <at 1 1.5462 RCIM_Cubic>]}}})
    Index 0 Param 2=&RandomRangeFloat(Minimum=0, Maximum=10)
    Index 0 Param 3=1
    Module.Index0Float0Write=true
    Module.Index0Float1Write=true
    Module.Index0Float2Write=true
    Module.Index0Float3Write=false
   )
   &ScaleColor(
    Scale Mode=::NewEnumerator1
    Scale RGBA=&MakeVector4_FromLinearColor(LinearColor=&FNTimeOfDayColor(Night Color=<rgba 0.2988 0.3514 0.5712 1>, Morning Color=<rgba 1 1 1 1>, Day Color=<rgba 1 1 1 1>, Evening Color=<rgba 1 1 1 1>))
    ScaleRGBA=true
   ) {module=ScaleColor001}
   /SetVariables(Particles.VisibilityTag=&MakeCustomIntFromBool(Bool=$User.bisHeadHidden, True INT=1, False INT=0)) {isSetParameters=true}
  ]
  rendererClasses [&NiagaraSpriteRendererProperties]
  renderers [{rendererClass=&NiagaraSpriteRendererProperties}]
 }
 Hair_Decal {
  emitterUpdateScript [
   &EmitterState(Life Cycle Mode=::NewEnumerator1, Loop Duration=0.5, UseLoopDelay=false)
   &SpawnBurst_Instantaneous(Use Spawn Probability=false, UseLoopCountLimit=false)
  ]
  particleSpawnScript [
   &InitializeParticle(
    Color=&MakeCustomLinearColorFromBool(
     Bool=&SetBoolByFloatComparison(A=$User.VariantID, B=0, Comparison Type=::NewEnumerator4)
     True Color=<rgba 0 0.566 0.7605 0.5>
     False Color=<rgba 1.2 0.7015 0.0069 0.8>
    )
    Sprite Rotation Angle=&RandomRangeFloat(Minimum=0.45, Maximum=0.55)
    Sprite Rotation Mode=::NewEnumerator1
    Sprite Size Max=<18 10>
    Sprite Size Min=<16 10>
    Sprite Size Mode=::NewEnumerator2
    UsePositionOffset=false
    Write Material Random=true
   )
   &CameraOffset(Camera Offset Amount=10)
   &AddVelocity(Velocity=<2 4 0>) {enabled=false}
  ]
  particleUpdateScript [
   &ParticleState()
   &SkeletalMeshLocation(
    MaskSpawnGroup=false
    Mesh Sampling Type=::NewEnumerator5
    Sampled Position Offset=&RandomRangeVector(Minimum=<-2 0 0>, Maximum=<0.5 0 0>)
    Skeletal Mesh=$System.Skeletal Mesh
    Socket Sampling Mode=::NewEnumerator5
   )
   &ScaleColor(
    Scale Alpha=&FloatFromCurve(FloatCurve=DataInterface() {propertyValues {Curve {
        defaultValue=1
        keys [
         {interpMode=RCIM_Cubic, tangentMode=RCTM_User, time=0, value=0}
         {interpMode=RCIM_Cubic, tangentMode=RCTM_User, time=0.5, value=1}
         {tangentMode=RCTM_User, time=1, value=0}
        ]
       }}})
    ScaleA=true
    ScaleRGB=true
   )
   &GravityForce(Gravity=<2 4 0>)
   &SolveForcesAndVelocity(Clamp Velocity=false, Limit Acceleration=false)
   &ScaleColor(
    Scale Mode=::NewEnumerator1
    Scale RGBA=&MakeVector4_FromLinearColor(LinearColor=&FNTimeOfDayColor(Night Color=<rgba 0.2288 0.2881 0.5712 1>, Morning Color=<rgba 1 1 1 1>, Day Color=<rgba 1 1 1 1>, Evening Color=<rgba 1 1 1 1>))
    ScaleRGBA=true
   ) {module=ScaleColor001}
   /SetVariables(Particles.VisibilityTag=&MakeCustomIntFromBool(Bool=$User.bisHeadHidden, True INT=1, False INT=0)) {isSetParameters=true}
  ]
  rendererClasses [&NiagaraSpriteRendererProperties]
  renderers [{rendererClass=&NiagaraSpriteRendererProperties}]
 }
}
summary {
 systemName=NS_TourBus_Hair_Hooded
 emitters [
  {emitterName=Hair_Decal, rendererClasses [&NiagaraSpriteRendererProperties]}
  {emitterName=Flames, rendererClasses [&NiagaraSpriteRendererProperties]}
 ]
 userVariables [
  User.Alpha {type=NiagaraFloat, defaultValue=0.5, description=""}
  User.bisHeadHidden {type=NiagaraBool}
  User.SpawnRate {type=NiagaraFloat, defaultValue=1, description=""}
  User.VariantID {type=NiagaraFloat, defaultValue=0, description=""}
 ]
}
