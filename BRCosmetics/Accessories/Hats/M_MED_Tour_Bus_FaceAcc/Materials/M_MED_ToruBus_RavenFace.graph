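# Two-sided lit material for the Raven face of the Tour Bus face accessory cosmetic, usable on skeletal meshes and mesh deformers.
# It builds material attributes from a single color parameter for base color, with zero metallic and specular and full roughness for a flat, matte, untextured look.
# Those attributes then pass through the shared MF_TechArtOverrides function.
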Material /BRCosmetics/Accessories/Hats/M_MED_Tour_Bus_FaceAcc/Materials/M_MED_ToruBus_RavenFace
graph {
 properties {
  bUseMaterialAttributes=true
  bUsedWithMeshDeformer=true
  bUsedWithSkeletalMesh=true
  pixelDepthOffsetMode=PDOM_Legacy
  twoSided=true
  usedShadingModels=Default Lit
 }
 nodes [
  n1 = Constant()
  n2 = SetMaterialAttributes!(MaterialAttributes, Base Color=VectorParameter().RGB, Metallic=@n1, Specular=@n1, Roughness=Constant() {r=1}) {attributeSetTypes [
    #1
    #2
    #3
    #4
   ]}
  MaterialFunctionCall(In=@n2) {materialFunction=&MF_TechArtOverrides}
 ]
}
