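# Masked material for the Tour Bus bandanna face accessory, with a noise-eroded reveal and animated emissive effects.
# BaseGrowth and EmisiveGrowth threshold inverted U minus tiled glitch noise, driving the opacity mask and a glowing band beyond the base.
# Kill and OnFire glows use sphere masks along U, broken up by panning noise, and add to emissive.
# Standard maps, a vertex-color hide WPO and MF_TechArtOverrides complete the attributes.
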Material /BRCosmetics/Accessories/Hats/M_MED_Tour_Bus_FaceAcc/Materials/M_MED_TourBus_Bandanna
graph {
 properties {
  bUseMaterialAttributes=true
  bUsedWithMeshDeformer=true
  bUsedWithSkeletalMesh=true
  blendMode=BLEND_Masked
  pixelDepthOffsetMode=PDOM_Legacy
  usedShadingModels=Default Lit
 }
 nodes [
  n1 = ComponentMask(None=TextureCoordinate()) {r=true}
  n2 = Add(A=OneMinus(None=@n1))
  n3 = Add(A=ScalarParameter() {parameterName=BaseGrowth}) {constB=0.5}
  n4 = MaterialFunctionCall(Float2=TextureCoordinate()) {materialFunction=&BreakOutFloat2Components}
  n5 = MaterialFunctionCall(
   X=Multiply(A=@n4.R, B=ScalarParameter() {defaultValue=12, parameterName=X_Tiling})
   Y=Multiply(A=@n4.G, B=ScalarParameter() {defaultValue=24, parameterName=Y_Tiling})
  ) {materialFunction=&MakeFloat2}
  n6 = Multiply(
   A=ChannelMaskParameter(None=TextureSampleParameter2D(UVs=@n5.Result) {parameterName=ErosionNoise, texture=/Game/Characters/Player/Male/Medium/Bodies/M_MED_Viper/Textures/Glitch/T_ViperGlitchNoise}.RGB) {parameterName=TearNoisechannel}.R
   B=ScalarParameter() {defaultValue=0.05, parameterName=NoiseSpread}
  )
  n7 = Clamp(None=Subtract(A=Subtract(A=@n2, B=@n3), B=@n6))
  n8 = Subtract(A=@n2, B=Add(A=ScalarParameter() {parameterName=EmisiveGrowth}) {constB=0.5})
  n9 = Round(None=Clamp(None=Subtract(A=@n8, B=@n6)))
  n10 = Max(A=Round(None=@n7), B=@n9)
  Clamp(None=@n10)
  n11 = TextureSample() {samplerType=SAMPLERTYPE_Masks, texture=/BRCosmetics/Accessories/Hats/M_MED_Tour_Bus_FaceAcc/Textures/T_M_MED_Tour_Bus_FaceAcc_S}
  n12 = TextureSampleParameter2D() {
   parameterName=Normal
   samplerType=SAMPLERTYPE_Normal
   texture=/BRCosmetics/Accessories/Hats/M_MED_Tour_Bus_FaceAcc/Textures/T_M_MED_Tour_Bus_FaceAcc_N
  }
  n13 = Multiply(A=Round(None=OneMinus(None=@n7)), B=@n9)
  n14 = VectorParameter() {parameterName=EmissiveGlowColor, defaultValue=<rgba 0 0.5915 1 1>}
  n15 = Power(Base=Multiply(A=Multiply(A=@n13, B=@n14.RGB), B=ScalarParameter() {defaultValue=10, parameterName=EmissiveGlowStrength}))
  n16 = MaterialFunctionCall(Float2=TextureCoordinate()) {materialFunction=&BreakOutFloat2Components}
  n17 = Multiply(A=@n16.R, B=ScalarParameter() {defaultValue=12, parameterName=KillNoise_X_Tiling})
  n18 = Multiply(A=@n16.G, B=ScalarParameter() {defaultValue=24, parameterName=KillNoiseY_Tiling})
  n19 = VectorParameter() {parameterName=EnergyPanSpeedAndDirection, defaultValue=<rgba 1 0 0 1>}
  n20 = TextureSampleParameter2D(UVs=Panner(
    Coordinate=MaterialFunctionCall(X=@n17, Y=@n18) {materialFunction=&MakeFloat2}.Result
    Speed=ComponentMask(None=@n19.RGB) {g=true, r=true}
   ) {speedX=1, speedY=1}) {parameterName=KillNoise, texture=&T_ViperGlitchNoise}
  n21 = Subtract(
   A=SphereMask(
    A=@n1
    B=ScalarParameter() {defaultValue=-0.5, parameterName=KillGlowControl}
    Radius=ScalarParameter() {defaultValue=0.2, parameterName=KillGlowRadius}
    Hardness=ScalarParameter() {parameterName=KillGlowHardness}
   )
   B=Multiply(
    A=ChannelMaskParameter(None=@n20.RGB) {parameterName=KillNoisechannel}.R
    B=ScalarParameter() {defaultValue=0.1, parameterName=CellNoiseStrength}
   )
  )
  n22 = Ceil(None=@n21)
  n23 = ScalarParameter() {parameterName=UseCeil}
  n24 = Multiply(
   A=VectorParameter() {parameterName=KillGlowColor, defaultValue=<rgba 1 0.4544 0 1>}.RGB
   B=Clamp(None=LinearInterpolate(A=@n21, B=@n22, Alpha=@n23))
  )
  n25 = ScalarParameter() {defaultValue=10, parameterName=KillGlowStrength}
  n26 = MaterialFunctionCall(Float2=TextureCoordinate()) {materialFunction=&BreakOutFloat2Components}
  n27 = Multiply(A=@n26.R, B=ScalarParameter() {defaultValue=12, parameterName=OnFireNoise_X_Tiling})
  n28 = Multiply(A=@n26.G, B=ScalarParameter() {defaultValue=24, parameterName=OnFireNoiseY_Tiling})
  n29 = VectorParameter() {parameterName=OnfirePanSpeedAndDirection, defaultValue=<rgba 1 0 0 1>}
  n30 = TextureSampleParameter2D(UVs=Panner(
    Coordinate=MaterialFunctionCall(X=@n27, Y=@n28) {materialFunction=&MakeFloat2}.Result
    Speed=ComponentMask(None=@n29.RGB) {g=true, r=true}
   ) {speedX=1, speedY=1}) {parameterName=KillNoise, texture=&T_ViperGlitchNoise}
  n31 = Subtract(
   A=SphereMask(
    A=@n1
    B=ScalarParameter() {defaultValue=-0.5, parameterName=OnFireGlowControl}
    Radius=ScalarParameter() {defaultValue=0.2, parameterName=OnFireGlowRadius}
    Hardness=ScalarParameter() {parameterName=OnFireGlowHardness}
   )
   B=Multiply(
    A=ChannelMaskParameter(None=@n30.RGB) {parameterName=KillNoisechannel}.R
    B=ScalarParameter() {defaultValue=0.1, parameterName=OnFireNoiseStrength}
   )
  )
  n32 = Ceil(None=@n31)
  n33 = ScalarParameter() {parameterName=UseCeil}
  n34 = Multiply(
   A=VectorParameter() {parameterName=OnFireGlowColor, defaultValue=<rgba 1 0.4544 0 1>}.RGB
   B=Clamp(None=LinearInterpolate(A=@n31, B=@n32, Alpha=@n33))
  )
  n35 = ScalarParameter() {defaultValue=10, parameterName=OnFireGlowStrength}
  n36 = Add(A=Power(Base=Multiply(A=@n24, B=@n25)), B=Power(Base=Multiply(A=@n34, B=@n35)))
  n37 = SetMaterialAttributes!(
   MaterialAttributes
   Base Color=NamedRerouteDeclaration(None=TextureSampleParameter2D() {parameterName=BaseColor, texture=/BRCosmetics/Accessories/Hats/M_MED_Tour_Bus_FaceAcc/Textures/T_M_MED_Tour_Bus_FaceAcc_D}.RGB) {name=baseColor, nodeColor=<rgba 1 0.3516 0 1>}
   Metallic=NamedRerouteDeclaration(None=@n11.G) {name=M, nodeColor=<rgba 0.5703 0 1 1>}
   Specular=NamedRerouteDeclaration(None=@n11.R) {name=S, nodeColor=<rgba 1 0 0.3047 1>}
   Roughness=NamedRerouteDeclaration(None=@n11.B) {name=R, nodeColor=<rgba 0.4062 0 1 1>}
   Normal=NamedRerouteDeclaration(None=@n12.RGB) {name=normal, nodeColor=<rgba 1 0.1406 0 1>}
   Opacity Mask=NamedRerouteDeclaration(None=@n10) {name=OpacityMask, nodeColor=<rgba 0.6172 1 0 1>}
   Emissive Color=NamedRerouteDeclaration(None=Add(A=@n15, B=@n36)) {name=emissive, nodeColor=<rgba 0.3594 1 0 1>}
   World Position Offset=NamedRerouteDeclaration(None=MaterialFunctionCall!(WPO) {materialFunction=/Game/Packages/Fortress_SharedMaterials/Base_Material_Functions/Fortnite_HideElementsWithVertexColor}.Result) {name=WPO, nodeColor=<rgba 0 1 0.6719 1>}
  ) {attributeSetTypes [
    #1
    #2
    #3
    #4
    #5
    #6
    #7
    #8
   ]}
  MaterialFunctionCall(In=@n37) {materialFunction=&MF_TechArtOverrides}
 ]
}
