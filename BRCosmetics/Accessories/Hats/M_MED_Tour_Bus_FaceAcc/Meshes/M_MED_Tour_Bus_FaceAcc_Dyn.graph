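# Physics asset for the M_MED_Tour_Bus_FaceAcc cosmetic hat or face accessory, defining ragdoll-style bodies for its dynamic top hair bones.
# A head capsule anchors a chain of three smaller capsules on dyn_hair_top_01 through 03.
# Each link is constrained to the bone before it with moderate swing and tight twist limits, so the chain can bend while barely rotating about its axis.
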PhysicsAsset /BRCosmetics/Accessories/Hats/M_MED_Tour_Bus_FaceAcc/Meshes/M_MED_Tour_Bus_FaceAcc_Dyn
bodies {
 dyn_hair_top_01 [{shapeType=Capsule, length=3, radius=2, rotation=<rot 90 0 0>}]
 dyn_hair_top_02 [{shapeType=Capsule, length=3, radius=2, rotation=<rot 90 0 0>}]
 dyn_hair_top_03 [{shapeType=Capsule, length=3, radius=2, rotation=<rot 90 0 0>}]
 head [{shapeType=Capsule, length=5, radius=5, rotation=<rot 90 0 0>}]
}
constraints [
 {bone1Name=dyn_hair_top_01, bone2Name=head, swing1LimitDegrees=45, swing2LimitDegrees=45, twistLimitDegrees=5}
 {bone1Name=dyn_hair_top_02, bone2Name=dyn_hair_top_01, swing1LimitDegrees=45, swing2LimitDegrees=45, twistLimitDegrees=5}
 {bone1Name=dyn_hair_top_03, bone2Name=dyn_hair_top_02, swing1LimitDegrees=45, swing2LimitDegrees=45, twistLimitDegrees=5}
]
