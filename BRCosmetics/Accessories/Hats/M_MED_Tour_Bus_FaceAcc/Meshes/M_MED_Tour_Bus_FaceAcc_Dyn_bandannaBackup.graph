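# Backup-named physics asset for the M_MED_Tour_Bus face accessory, defining the bodies for its dynamic bandanna tails.
# Left and right capsule chains branch from a master bone on the head, with tight twist limits and swing limits that generally widen toward the tips.
# Capsules on the head, neck, spine, pelvis, clavicles and upper arms approximate the wearer's body, linked mostly by constraints with free swing and twist.
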PhysicsAsset /BRCosmetics/Accessories/Hats/M_MED_Tour_Bus_FaceAcc/Meshes/M_MED_Tour_Bus_FaceAcc_Dyn_bandannaBackup
bodies {
 clavicle_l [{shapeType=Capsule, length=8.6189, radius=8.0726, center=<11.5493 0.0008 0>, rotation=<rot 78.3207 0 0>}]
 clavicle_r [{shapeType=Capsule, length=8.6364, radius=8.0726, center=<-11.5493 -0.0008 0>, rotation=<rot 78.3207 0 0>}]
 dyn_bandanna_MSTR [{shapeType=Capsule, length=0.5, radius=0.5}]
 dyn_bandanna_l_01 [{shapeType=Capsule, length=5.25, radius=1, shapeName=dyn_bandanna_l_01_capsule, center=<3.4718 0 0.0002>, rotation=<rot 90 0 0>}]
 dyn_bandanna_l_02 [{shapeType=Capsule, length=4.9732, radius=0.9569, center=<2.4372 0.0002 0>, rotation=<rot 90 0 0>}]
 dyn_bandanna_l_02a [{shapeType=Capsule, length=4.3712, radius=0.9569, center=<2.4372 0.0002 0>, rotation=<rot 90 0 0>}]
 dyn_bandanna_l_03 [{shapeType=Capsule, length=8.5, radius=1, center=<5.5 0 0>, rotation=<rot 90 0 0>}]
 dyn_bandanna_l_04 [{shapeType=Capsule, length=8.5, radius=1, center=<5.5 0 0>, rotation=<rot 90 0 0>}]
 dyn_bandanna_l_05 [{shapeType=Capsule, length=8.5, radius=1.2, center=<5.5 0 0>, rotation=<rot 90 0 0>}]
 dyn_bandanna_l_06 [{shapeType=Capsule, length=8.5, radius=1.2, center=<5.5 0 0>, rotation=<rot 90 0 0>}]
 dyn_bandanna_l_07 [{shapeType=Capsule, length=8.5, radius=1.6, center=<5.5 0 0>, rotation=<rot 90 0 0>}]
 dyn_bandanna_l_08 [{shapeType=Capsule, length=2.0048, radius=1.6, center=<4 0 0>, rotation=<rot 90 0 0>}]
 dyn_bandanna_r_01 [{shapeType=Capsule, length=5.5, radius=1, shapeName=dyn_bandanna_r_01_capsule, center=<3.5735 0 0>, rotation=<rot 90 0 0>}]
 dyn_bandanna_r_02 [{shapeType=Capsule, length=4.9732, radius=0.9569, center=<2.4372 0.0002 0>, rotation=<rot 90 0 0>}]
 dyn_bandanna_r_02a [{shapeType=Capsule, length=4.3712, radius=0.9569, center=<2.4372 0.0002 0>, rotation=<rot 90 0 0>}]
 dyn_bandanna_r_03 [{shapeType=Capsule, length=8.5, radius=1, center=<5.5 0 0>, rotation=<rot 90 0 0>}]
 dyn_bandanna_r_04 [{shapeType=Capsule, length=8.5, radius=1, center=<5.5 0 0>, rotation=<rot 90 0 0>}]
 dyn_bandanna_r_05 [{shapeType=Capsule, length=8.5, radius=1.2, center=<5.5 0 0>, rotation=<rot 90 0 0>}]
 dyn_bandanna_r_06 [{shapeType=Capsule, length=8.5, radius=1.2, center=<5.5 0 0>, rotation=<rot 90 0 0>}]
 dyn_bandanna_r_07 [{shapeType=Capsule, length=8.5, radius=1.6, center=<5.5 0 0>, rotation=<rot 90 0 0>}]
 dyn_bandanna_r_08 [{shapeType=Capsule, length=2.0048, radius=1.6, center=<4 0 0>, rotation=<rot 90 0 0>}]
 head [{shapeType=Capsule, length=8.6093, radius=7.2658, center=<0 1.6805 0>, rotation=<rot 90 0 0>}]
 neck_02 [{shapeType=Capsule, length=8.6093, radius=7.2658, center=<-0.6233 0.7996 -0.0001>, rotation=<rot 90 0 -10.5475>}]
 pelvis [{shapeType=Capsule, length=6.5541, radius=16.4378, center=<-4.4939 -0.8081 0>}]
 spine_01 [{shapeType=Capsule, length=3.0512, radius=15.311, center=<-2.4241 0 0>}]
 spine_02 [{shapeType=Capsule, length=6.2793, radius=11.7961, center=<-0.907 -0.319 0.0002>}]
 spine_03 [{shapeType=Capsule, length=10.2116, radius=11.7961, center=<1.7158 0 0>}]
 spine_04 [{shapeType=Capsule, length=11.7529, radius=13.5512, center=<1.7158 1.362 0.0002>}]
 spine_05 [
  {shapeType=Capsule, length=19.1828, radius=14.5, center=<-1.6727 1.4075 0>}
  {shapeType=Capsule, length=5.0764, radius=11.9971, center=<9.9207 0.8174 0.0001>}
 ]
 upperarm_l [{shapeType=Capsule, length=21.3832, radius=6.3586, center=<13.4948 0.0004 0>, rotation=<rot 90 0 0>}]
 upperarm_r [{shapeType=Capsule, length=21.3832, radius=6.3586, center=<-13.4948 -0.0004 0>, rotation=<rot 90 0 0>}]
}
constraints [
 {bone1Name=head, bone2Name=neck_02, swing1LimitDegrees=45, swing2LimitDegrees=45, twistLimitDegrees=45}
 {
  bone1Name=dyn_bandanna_MSTR
  bone2Name=head
  swing1LimitDegrees=45
  swing1Motion=Free
  swing2LimitDegrees=45
  swing2Motion=Free
  twistLimitDegrees=45
  twistMotion=Free
 }
 {bone1Name=dyn_bandanna_l_01, bone2Name=dyn_bandanna_MSTR, swing1LimitDegrees=10, swing2LimitDegrees=40, twistLimitDegrees=5}
 {bone1Name=dyn_bandanna_l_02, bone2Name=dyn_bandanna_l_01, swing1LimitDegrees=20, swing2LimitDegrees=20, twistLimitDegrees=5}
 {bone1Name=dyn_bandanna_l_03, bone2Name=dyn_bandanna_l_02a, swing1LimitDegrees=20, swing2LimitDegrees=20, twistLimitDegrees=5}
 {bone1Name=dyn_bandanna_l_04, bone2Name=dyn_bandanna_l_03, swing1LimitDegrees=20, swing2LimitDegrees=20, twistLimitDegrees=5}
 {bone1Name=dyn_bandanna_l_05, bone2Name=dyn_bandanna_l_04, swing1LimitDegrees=45, swing2LimitDegrees=45, twistLimitDegrees=5}
 {bone1Name=dyn_bandanna_l_06, bone2Name=dyn_bandanna_l_05, swing1LimitDegrees=45, swing2LimitDegrees=45, twistLimitDegrees=5}
 {bone1Name=dyn_bandanna_l_07, bone2Name=dyn_bandanna_l_06, swing1LimitDegrees=45, swing2LimitDegrees=45, twistLimitDegrees=5}
 {bone1Name=dyn_bandanna_l_08, bone2Name=dyn_bandanna_l_07, swing1LimitDegrees=45, swing2LimitDegrees=45, twistLimitDegrees=5}
 {bone1Name=dyn_bandanna_r_01, bone2Name=dyn_bandanna_MSTR, swing1LimitDegrees=10, swing2LimitDegrees=40, twistLimitDegrees=5}
 {bone1Name=dyn_bandanna_r_02, bone2Name=dyn_bandanna_r_01, swing1LimitDegrees=20, swing2LimitDegrees=20, twistLimitDegrees=5}
 {bone1Name=dyn_bandanna_r_03, bone2Name=dyn_bandanna_r_02a, swing1LimitDegrees=20, swing2LimitDegrees=20, twistLimitDegrees=5}
 {bone1Name=dyn_bandanna_r_04, bone2Name=dyn_bandanna_r_03, swing1LimitDegrees=20, swing2LimitDegrees=20, twistLimitDegrees=5}
 {bone1Name=dyn_bandanna_r_05, bone2Name=dyn_bandanna_r_04, swing1LimitDegrees=45, swing2LimitDegrees=45, twistLimitDegrees=5}
 {bone1Name=dyn_bandanna_r_06, bone2Name=dyn_bandanna_r_05, swing1LimitDegrees=45, swing2LimitDegrees=45, twistLimitDegrees=5}
 {bone1Name=dyn_bandanna_r_07, bone2Name=dyn_bandanna_r_06, swing1LimitDegrees=45, swing2LimitDegrees=45, twistLimitDegrees=5}
 {bone1Name=dyn_bandanna_r_08, bone2Name=dyn_bandanna_r_07, swing1LimitDegrees=45, swing2LimitDegrees=45, twistLimitDegrees=5}
 {
  bone1Name=spine_05
  bone2Name=spine_04
  swing1LimitDegrees=45
  swing1Motion=Free
  swing2LimitDegrees=45
  swing2Motion=Free
  twistLimitDegrees=45
  twistMotion=Free
 }
 {
  bone1Name=spine_04
  bone2Name=spine_03
  swing1LimitDegrees=45
  swing1Motion=Free
  swing2LimitDegrees=45
  swing2Motion=Free
  twistLimitDegrees=45
  twistMotion=Free
 }
 {
  bone1Name=spine_03
  bone2Name=spine_02
  swing1LimitDegrees=45
  swing1Motion=Free
  swing2LimitDegrees=45
  swing2Motion=Free
  twistLimitDegrees=45
  twistMotion=Free
 }
 {
  bone1Name=spine_02
  bone2Name=spine_01
  swing1LimitDegrees=45
  swing1Motion=Free
  swing2LimitDegrees=45
  swing2Motion=Free
  twistLimitDegrees=45
  twistMotion=Free
 }
 {
  bone1Name=spine_01
  bone2Name=pelvis
  swing1LimitDegrees=45
  swing1Motion=Free
  swing2LimitDegrees=45
  swing2Motion=Free
  twistLimitDegrees=45
  twistMotion=Free
 }
 {
  bone1Name=neck_02
  bone2Name=spine_05
  swing1LimitDegrees=45
  swing1Motion=Free
  swing2LimitDegrees=45
  swing2Motion=Free
  twistLimitDegrees=45
  twistMotion=Free
 }
 {
  bone1Name=clavicle_l
  bone2Name=spine_05
  swing1LimitDegrees=45
  swing1Motion=Free
  swing2LimitDegrees=45
  swing2Motion=Free
  twistLimitDegrees=45
  twistMotion=Free
 }
 {
  bone1Name=clavicle_r
  bone2Name=spine_05
  swing1LimitDegrees=45
  swing1Motion=Free
  swing2LimitDegrees=45
  swing2Motion=Free
  twistLimitDegrees=45
  twistMotion=Free
 }
 {
  bone1Name=upperarm_l
  bone2Name=clavicle_l
  swing1LimitDegrees=45
  swing1Motion=Free
  swing2LimitDegrees=45
  swing2Motion=Free
  twistLimitDegrees=45
  twistMotion=Free
 }
 {
  bone1Name=upperarm_r
  bone2Name=clavicle_r
  swing1LimitDegrees=45
  swing1Motion=Free
  swing2LimitDegrees=45
  swing2Motion=Free
  twistLimitDegrees=45
  twistMotion=Free
 }
 {bone1Name=dyn_bandanna_l_02a, bone2Name=dyn_bandanna_l_02, swing1LimitDegrees=10, swing2LimitDegrees=10, twistLimitDegrees=5}
 {bone1Name=dyn_bandanna_r_02a, bone2Name=dyn_bandanna_r_02, swing1LimitDegrees=10, swing2LimitDegrees=10, twistLimitDegrees=5}
]
