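# Animation Blueprint for the M_MED_Tour_Bus_FaceAcc cosmetic, adding bandanna physics on top of the pose copied from the attached parent mesh.
# A neck-relative RigidBody pass, plus a small head offset at LOD1, feeds a cached pose.
# Two spine-relative RigidBody passes use that cache with different physics assets, selected by whether backbling is worn.
# Their gravity is normal except during skydiving before parachute opening, when it turns upward or zero depending on vertical velocity.
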AnimBlueprint /BRCosmetics/Accessories/Hats/M_MED_Tour_Bus_FaceAcc/Meshes/M_MED_Tour_Bus_FaceAcc_AnimBP
target_skeleton=&/M_MED_Tour_Bus_FaceAcc_Skeleton
graphs [
 AnimGraph {nodes [
   n1 = Get gravity_override() -> (gravity_override:Vector)
   n2 = RigidBody!(ComponentPose=Local To Component(LocalPose=Use cached pose 'basePose'()), Alpha, OverrideWorldGravity:Vector=@n1, ExternalForce) -> (Pose) {
    bClampLinearTranslationLimitToRefPose=true
    bOverrideWorldGravity=true
    overridePhysicsAsset=&/M_MED_Tour_Bus_FaceAcc_Dyn_bandanna_withBackblingJiraUpdates
    simulationSpace=BaseBoneSpace
    alphaBoolBlend {blendOption=Linear}
    baseBoneRef {boneName=spine_05}
    componentAppliedLinearAccClamp=<3000 3000 3000>
    componentLinearAccScale=<0.5 0.4 0.4>
    componentLinearVelScale=<0.5 0.4 0.4>
    overrideWorldGravity {z=-980}
   }
   n3 = RigidBody(ComponentPose=Local To Component(LocalPose=Use cached pose 'basePose'()), OverrideWorldGravity=@n1) {
    bClampLinearTranslationLimitToRefPose=true
    bOverrideWorldGravity=true
    overridePhysicsAsset=&/M_MED_Tour_Bus_FaceAcc_Dyn_bandannaBackup
    simulationSpace=BaseBoneSpace
    alphaBoolBlend {blendOption=Linear}
    baseBoneRef {boneName=spine_05}
    componentAppliedLinearAccClamp=<3000 3000 3000>
    componentLinearAccScale=<0.5 0.4 0.4>
    componentLinearVelScale=<0.5 0.4 0.4>
    overrideWorldGravity {z=-980}
   }
   Output Pose(Result=Blend Poses by bool(BlendPose_0=Component To Local(ComponentPose=@n2), BlendPose_1=Component To Local(ComponentPose=@n3), bActiveValue=Get isWearingBackbling()) {blendType=Linear}) {layerGroup=None}
   n4 = RigidBody!(ComponentPose=Local To Component(LocalPose=Copy Pose From Mesh() {bUseAttachedParent=true}), Alpha, ExternalForce) -> (Pose) {
    bClampLinearTranslationLimitToRefPose=true
    bOverrideWorldGravity=true
    overridePhysicsAsset=&/M_MED_Tour_Bus_FaceAcc_Dyn
    simulationSpace=BaseBoneSpace
    alphaBoolBlend {blendOption=Linear}
    baseBoneRef {boneName=neck_01}
    componentAppliedLinearAccClamp=<3000 3000 3000>
    componentLinearAccScale=<0.5 0.4 0.4>
    componentLinearVelScale=<0.5 0.4 0.4>
    overrideWorldGravity {z=-980}
   }
   n5 = "Transform (Modify) Bone - Bone: head"!(
    ComponentPose=@n4
    bAlphaBoolEnabled:bool=Get Mask_move_LOD1() -> (Mask_move_LOD1:bool)
    Translation="0.000,0.1,-0.000"
    Rotation="0,0,0"
    Scale="1,1,1"
   ) -> (Pose) {alphaInputType=Bool, translationMode=BMM_Additive, alphaBoolBlend {blendOption=Linear}, boneToModify {boneName=head}, translation {y=0.1}}
   Save cached pose 'basePose'(Pose=Component To Local(ComponentPose=@n5)) {cacheName=basePose}
  ]}
 EventGraph {nodes [
   Event BlueprintUpdateAnimation()
   > Set Mask_move_LOD1(Mask_move_LOD1:bool="Equal (Integer)"(A=Get CurrentLOD(), B=1)) -> (Output_Get:bool)
   > n1 = Branch(Condition=Get bIsSkyDiving())
   > n2 = Branch(Condition=Get bIsParachuteOpened())
   Set gravity_override(execute=[@n1.else, @n2.then], gravity_override="0, 0,-980")
   n3 = Branch(execute=@n2.else, Condition="float > float"(A=Get PawnVelocityZ(), B=50) {is=title K2Node_CallFunction})
   > Set gravity_override(gravity_override="0, 0,400")
   Set gravity_override(execute=@n3.else, gravity_override="0, 0, 0")
   Event OnCharacterCustomizationLoaded()
   > Set isWearingBackbling(isWearingBackbling=CheckForBackbling())
   TryGetPawnOwner()
  ]}
]
variables [
 gravity_override {type=Vector}
 isWearingBackbling {type=boolean}
 Mask_move_LOD1 {type=boolean}
]
